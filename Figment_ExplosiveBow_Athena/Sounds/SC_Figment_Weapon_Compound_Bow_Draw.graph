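# Sound cue for drawing the Figment explosive compound bow, played as weapon foley with reload attenuation.
# A Mixer layers one of three equally weighted random draw one-shots with a quieter full-tension loop.
# The loop starts after a fixed delay and runs through an Enveloper with a volume fade-in curve.
# A root Modulator wraps the mix, and the cue overrides concurrency and has its debug flag on.
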SoundCue /Figment_ExplosiveBow_Athena/Sounds/SC_Figment_Weapon_Compound_Bow_Draw
root=Modulator_0
nodes [
 {id=Delay_0, children [WavePlayer_1], properties {delayMax=2, delayMin=2}}
 {
  id=Enveloper_0
  children [Delay_0]
  properties {pitchCurve {editorCurveData {keys [{time=0}]}}, volumeCurve {editorCurveData {keys [<at 0 0>,<at 0.5 1>]}}}
 }
 {id=Mixer_0, children [Random_0,Enveloper_0], properties {inputVolume [1,0.5]}}
 {id=Modulator_0, children [Mixer_0]}
 {id=Random_0, children [WavePlayer_0,WavePlayer_3,WavePlayer_2], properties {weights [1,1,1]}}
 {id=WavePlayer_0, properties {soundWaveAssetPtr=/Game/Sounds/Fort_Weapons/Explosive_Bow/Weapon_Bow_Compound_Draw_02}}
 {id=WavePlayer_1, properties {bLooping=true, soundWaveAssetPtr=/Game/Sounds/Fort_Weapons/Explosive_Bow/Weapon_Bow_Compound_Full_Tension_Loop}}
 {id=WavePlayer_2, properties {soundWaveAssetPtr=/Game/Sounds/Fort_Weapons/Explosive_Bow/Weapon_Bow_Compound_Draw_01}}
 {id=WavePlayer_3, properties {soundWaveAssetPtr=/Game/Sounds/Fort_Weapons/Explosive_Bow/Weapon_Bow_Compound_Draw_03}}
]
properties {
 attenuationSettings=&Fort_Weapon_Reload_Att
 bDebug=true
 bOverrideConcurrency=true
 maxDistance=2250
 soundClassObject=&Fort_WeaponFoley
 volumeMultiplier=0.376
}
