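# Sound cue for the Figment explosive compound bow's empty sound, classed as weapon foley.
# It plays a single wave, assault_end_sdk_01, through a modulator that fixes a raised pitch, at reduced cue volume.
# A concurrency override allows only one instance at a time and blocks new plays rather than cutting off the current one.
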SoundCue /Figment_ExplosiveBow_Athena/Sounds/SC_Figment_Weapon_Compound_Bow_Empty
root=Modulator_4
nodes [
 {id=Modulator_4, children [WavePlayer_0], properties {pitchMax=1.3, pitchMin=1.3}}
 {id=WavePlayer_0, properties {soundWaveAssetPtr=/Figment_ExplosiveBow_Athena/Sounds/SoundWavs/assault_end_sdk_01}}
]
properties {
 bOverrideConcurrency=true
 maxDistance=2097152
 soundClassObject=&Fort_WeaponFoley
 volumeMultiplier=0.251
 attenuationOverrides {falloffDistance=1600}
 concurrencyOverrides {maxCount=1, resolutionRule=PreventNew}
}
